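# Control Rig for the M_LRG_QueenTruth_Rind player body that procedurally animates its mech and "Bart" add-on bones atop the shared DeformRig function.
# Bart arm controls blend between offsets from movement, aim and DeformRig attributes, and Hermite splines through them drive the arm chains.
# Swing-twist thigh and head rotations, remapped through curves, rotate Bart's legs while crouching and the mech neck piece.
# The back piece follows a weighted five-bone spine constraint.
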ControlRigBlueprint /BRCosmetics/Characters/Player/Male/Large/Bodies/M_LRG_QueenTruth_Rind/Meshes/M_LRG_QueenTruth_Rind_CtrlRig
graphs [
 New Function_ContainedGraph {nodes [
   RigVMFunctionEntryNode()
   > RigVMFunction_Sequence = RigVMAggregateNode()
   RigVMFunctionReturnNode(ExecuteContext=@RigVMFunction_Sequence.C)
   n1 = &RigUnit_GetRelativeTransformForItem(Child=(Type=Bone,Name="thigh_twist_01_l"), Parent=(Type=Control,Name="thigh_twist_01_l_ctrl"), bChildInitial=False, bParentInitial=False)
   RigVMVariableNode(
    Value=&RigVMFunction_MathQuaternionToEuler(
     RotationOrder=ZYX
     Value=&RigVMFunction_MathQuaternionSwingTwist(TwistAxis=(X=1,Y=0,Z=0), Input=@n1.RelativeTransform.Rotation).Swing
    ).Result
    ExecuteContext=@RigVMFunction_Sequence.A
   ) {variable=BartLegSwing_L}
   n2 = &RigUnit_GetRelativeTransformForItem(Child=(Type=Bone,Name="thigh_twist_01_r"), Parent=(Type=Control,Name="thigh_twist_01_r_ctrl"), bChildInitial=False, bParentInitial=False)
   > RigVMVariableNode(Value=&RigVMFunction_MathQuaternionToEuler(
     RotationOrder=ZYX
     Value=&RigVMFunction_MathQuaternionSwingTwist(TwistAxis=(X=1,Y=0,Z=0), Input=@n2.RelativeTransform.Rotation).Swing
    ).Result) {variable=BartLegSwing_R}
   n3 = RigVMVariableNode() {variable=BartLegSwing_L}
   n4 = &RigVMFunction_AnimEvalRichCurve(
    Curve="(EditorCurveData=(Keys=((InterpMode=RCIM_Cubic,TangentMode=RCTM_User,ArriveTangent=0.2796,LeaveTangent=0.2796),(InterpMode=RCIM_Cubic,TangentMode=RCTM_User,Time=0.5233,Value=0.5785,ArriveTangent=1.2761,LeaveTangent=1.2761),(InterpMode=RCIM_Cubic,TangentMod ...(truncated)"
    SourceMaximum=-140
    SourceMinimum=0
    TargetMaximum=-44
    TargetMinimum=0
    Value=@n3.Value.Y
   )
   n5 = &RigVMFunction_AnimEvalRichCurve(
    Curve="(EditorCurveData=(Keys=((InterpMode=RCIM_Cubic,TangentMode=RCTM_User,ArriveTangent=0.2796,LeaveTangent=0.2796),(InterpMode=RCIM_Cubic,TangentMode=RCTM_User,Time=0.5554,Value=0.5798,ArriveTangent=1.2761,LeaveTangent=1.2761),(InterpMode=RCIM_Cubic,TangentMod ...(truncated)"
    SourceMaximum=-140
    SourceMinimum=0
    TargetMaximum=-44
    TargetMinimum=0
    Value=@n3.Value.Z
   )
   n6 = &RigVMFunction_AlphaInterp(
    InterpSpeedDecreasing=10
    InterpSpeedIncreasing=10
    Scale=1
    bClampResult=false
    bInterpResult=false
    bMapRange=true
    Value=&RigVMFunction_MathDoubleAdd(A=@n4.Result, B=@n5.Result).Result
   )
   &RigUnit_ModifyTransforms(
    ItemToModify=((Item=(Type=Bone,Name="base_BartLeg_l"),Transform=(Rotation=(X=0,Y=0,Z=0,W=1),Translation=(X=0,Y=0,Z=0),Scale3D=(X=1,Y=1,Z=1))))
    WeightMinimum=0
    Weight=&RigVMFunction_MathBoolToFloat(Value=RigVMVariableNode() {variable=isCrouching}.Value).Result
    ItemToModify.0.Transform.Rotation=&RigVMFunction_MathQuaternionFromEuler(Euler=(X=0,Y=0,Z=0), Euler.Y=&RigVMFunction_MathDoubleMul(B=2.5, A=@n6.Result).Result).Result
    ExecutePin=@RigVMFunction_Sequence.B
   )
   RigVMFunction_MathBoolToFloat = &RigVMFunction_MathBoolToFloat(Value=RigVMVariableNode() {variable=isCrouching}.Value)
   n7 = RigVMVariableNode() {variable=BartLegSwing_R}
   n8 = &RigVMFunction_AnimEvalRichCurve(
    Curve="(EditorCurveData=(Keys=((InterpMode=RCIM_Cubic,TangentMode=RCTM_User,ArriveTangent=0.2796,LeaveTangent=0.2796),(InterpMode=RCIM_Cubic,TangentMode=RCTM_User,Time=0.5233,Value=0.5785,ArriveTangent=1.2761,LeaveTangent=1.2761),(InterpMode=RCIM_Cubic,TangentMod ...(truncated)"
    SourceMaximum=-140
    SourceMinimum=0
    TargetMaximum=-44
    TargetMinimum=0
    Value=@n7.Value.Y
   )
   n9 = &RigVMFunction_AnimEvalRichCurve(
    Curve="(EditorCurveData=(Keys=((InterpMode=RCIM_Cubic,TangentMode=RCTM_User,ArriveTangent=0.2796,LeaveTangent=0.2796),(InterpMode=RCIM_Cubic,TangentMode=RCTM_User,Time=0.5554,Value=0.5798,ArriveTangent=1.2761,LeaveTangent=1.2761),(InterpMode=RCIM_Cubic,TangentMod ...(truncated)"
    SourceMaximum=-140
    SourceMinimum=0
    TargetMaximum=-44
    TargetMinimum=0
    Value=@n7.Value.Z
   )
   n10 = &RigVMFunction_AlphaInterp(
    InterpSpeedDecreasing=10
    InterpSpeedIncreasing=10
    Scale=1
    bClampResult=false
    bInterpResult=false
    bMapRange=true
    Value=&RigVMFunction_MathDoubleAdd(A=@n8.Result, B=@n9.Result).Result
   )
   > &RigUnit_ModifyTransforms(
    ItemToModify=((Item=(Type=Bone,Name="base_BartLeg_r"),Transform=(Rotation=(X=0,Y=0,Z=0,W=1),Translation=(X=0,Y=0,Z=0),Scale3D=(X=1,Y=1,Z=1))))
    WeightMinimum=0
    Weight=@RigVMFunction_MathBoolToFloat.Result
    ItemToModify.0.Transform.Rotation=&RigVMFunction_MathQuaternionFromEuler(Euler=(X=0,Y=0,Z=0), Euler.Y=&RigVMFunction_MathDoubleMul(B=3, A=@n10.Result).Result).Result
   )
  ]}
 New Function_ContainedGraph {nodes [
   RigVMFunctionEntryNode()
   > Control Stack at Item = RigVMFunctionReferenceNode(
    Add Bottom Null=False
    Add Null=False
    Add Secondary Control=False
    Control Suffix=_ctrl
    Item=(Name="thigh_twist_01_l",Type=Bone)
    Null Suffix=_null
    Parent=(Name="pelvis",Type=Bone)
    Secondary Settings="(Shape=(bVisible=True,Name=\"Default\",Color=(R=1,G=0,B=0,A=1),Transform=(Rotation=(X=0,Y=0,Z=0,W=1),Translation=(X=0,Y=0,Z=0),Scale3D=(X=1,Y=1,Z=1))),Proxy=(bIsProxy=False,ShapeVisibility=BasedOnSelection),DisplayName=\"None\",InitialSpace=LocalSpace,bUsePref ...(truncated)"
    Settings="(Shape=(bVisible=True,Name=\"Default\",Color=(R=0,G=0,B=1,A=1),Transform=(Rotation=(X=0,Y=0,Z=0,W=1),Translation=(X=0,Y=0,Z=0),Scale3D=(X=1,Y=1,Z=1))),Proxy=(bIsProxy=False,ShapeVisibility=BasedOnSelection),DisplayName=\"None\",InitialSpace=LocalSpace,bUsePref ...(truncated)"
    Tag Controls=true
    Zero Controls=true
   )
   > Control Stack at Item_1 = RigVMFunctionReferenceNode(
    Add Bottom Null=False
    Add Null=False
    Add Secondary Control=False
    Control Suffix=_ctrl
    Item=(Name="thigh_twist_01_r",Type=Bone)
    Null Suffix=_null
    Parent=(Name="pelvis",Type=Bone)
    Secondary Settings="(Shape=(bVisible=True,Name=\"Default\",Color=(R=1,G=0,B=0,A=1),Transform=(Rotation=(X=0,Y=0,Z=0,W=1),Translation=(X=0,Y=0,Z=0),Scale3D=(X=1,Y=1,Z=1))),Proxy=(bIsProxy=False,ShapeVisibility=BasedOnSelection),DisplayName=\"None\",InitialSpace=LocalSpace,bUsePref ...(truncated)"
    Settings="(Shape=(bVisible=True,Name=\"Default\",Color=(R=1,G=0,B=0,A=1),Transform=(Rotation=(X=0,Y=0,Z=0,W=1),Translation=(X=0,Y=0,Z=0),Scale3D=(X=1,Y=1,Z=1))),Proxy=(bIsProxy=False,ShapeVisibility=BasedOnSelection),DisplayName=\"None\",InitialSpace=LocalSpace,bUsePref ...(truncated)"
    Tag Controls=true
    Zero Controls=true
   )
   > Control Stack at Item_8 = RigVMFunctionReferenceNode(
    Add Bottom Null=False
    Add Null=true
    Add Secondary Control=False
    Control Suffix=_ctrl
    Item=(Name="dyn_BartClavicle_l",Type=Bone)
    Null Suffix=_null
    Parent=(Name="spine_05",Type=Bone)
    Secondary Settings="(Shape=(bVisible=True,Name=\"Default\",Color=(R=1,G=0,B=0,A=1),Transform=(Rotation=(X=0,Y=0,Z=0,W=1),Translation=(X=0,Y=0,Z=0),Scale3D=(X=1,Y=1,Z=1))),Proxy=(bIsProxy=False,ShapeVisibility=BasedOnSelection),DisplayName=\"None\",InitialSpace=LocalSpace,bUsePref ...(truncated)"
    Settings="(Shape=(bVisible=True,Name=\"Circle_Thick\",Color=(R=0,G=0,B=1,A=1),Transform=(Rotation=(X=0,Y=-0.7071,Z=0,W=0.7071),Translation=(X=0,Y=0,Z=0),Scale3D=(X=1,Y=1,Z=1))),Proxy=(bIsProxy=False,ShapeVisibility=BasedOnSelection),DisplayName=\"None\",InitialSpace=Loc ...(truncated)"
    Tag Controls=true
    Zero Controls=true
   )
   > Control Stack at Item_2 = RigVMFunctionReferenceNode(
    Add Bottom Null=False
    Add Null=true
    Add Secondary Control=False
    Control Suffix=_ctrl
    Item=(Name="dyn_BartHand_l",Type=Bone)
    Null Suffix=_null
    Parent=(Name="spine_05",Type=Bone)
    Secondary Settings="(Shape=(bVisible=True,Name=\"Default\",Color=(R=1,G=0,B=0,A=1),Transform=(Rotation=(X=0,Y=0,Z=0,W=1),Translation=(X=0,Y=0,Z=0),Scale3D=(X=1,Y=1,Z=1))),Proxy=(bIsProxy=False,ShapeVisibility=BasedOnSelection),DisplayName=\"None\",InitialSpace=LocalSpace,bUsePref ...(truncated)"
    Settings="(Shape=(bVisible=True,Name=\"Circle_Thick\",Color=(R=0,G=0,B=1,A=1),Transform=(Rotation=(X=0,Y=-0.7071,Z=0,W=0.7071),Translation=(X=0,Y=0,Z=0),Scale3D=(X=1,Y=1,Z=1))),Proxy=(bIsProxy=False,ShapeVisibility=BasedOnSelection),DisplayName=\"None\",InitialSpace=Loc ...(truncated)"
    Tag Controls=true
    Zero Controls=true
   )
   > Control Stack at Item_3 = RigVMFunctionReferenceNode(
    Add Bottom Null=False
    Add Null=true
    Add Secondary Control=False
    Control Suffix=_ctrl
    Item=(Name="dyn_BartLowerarm_l",Type=Bone)
    Null Suffix=_null
    Parent=(Name="spine_05",Type=Bone)
    Secondary Settings="(Shape=(bVisible=True,Name=\"Default\",Color=(R=1,G=0,B=0,A=1),Transform=(Rotation=(X=0,Y=0,Z=0,W=1),Translation=(X=0,Y=0,Z=0),Scale3D=(X=1,Y=1,Z=1))),Proxy=(bIsProxy=False,ShapeVisibility=BasedOnSelection),DisplayName=\"None\",InitialSpace=LocalSpace,bUsePref ...(truncated)"
    Settings="(Shape=(bVisible=True,Name=\"Circle_Thick\",Color=(R=0,G=0,B=1,A=1),Transform=(Rotation=(X=0,Y=-0.7071,Z=0,W=0.7071),Translation=(X=0,Y=0,Z=0),Scale3D=(X=1,Y=1,Z=1))),Proxy=(bIsProxy=False,ShapeVisibility=BasedOnSelection),DisplayName=\"None\",InitialSpace=Loc ...(truncated)"
    Tag Controls=true
    Zero Controls=true
   )
   > Control Stack at Item_4 = RigVMFunctionReferenceNode(
    Add Bottom Null=False
    Add Null=true
    Add Secondary Control=False
    Control Suffix=_ctrl
    Item=(Name="dyn_BartUpperarm_l",Type=Bone)
    Null Suffix=_null
    Parent=(Name="spine_05",Type=Bone)
    Secondary Settings="(Shape=(bVisible=True,Name=\"Default\",Color=(R=1,G=0,B=0,A=1),Transform=(Rotation=(X=0,Y=0,Z=0,W=1),Translation=(X=0,Y=0,Z=0),Scale3D=(X=1,Y=1,Z=1))),Proxy=(bIsProxy=False,ShapeVisibility=BasedOnSelection),DisplayName=\"None\",InitialSpace=LocalSpace,bUsePref ...(truncated)"
    Settings="(Shape=(bVisible=True,Name=\"Circle_Thick\",Color=(R=0,G=0,B=1,A=1),Transform=(Rotation=(X=0,Y=-0.7071,Z=0,W=0.7071),Translation=(X=0,Y=0,Z=0),Scale3D=(X=1,Y=1,Z=1))),Proxy=(bIsProxy=False,ShapeVisibility=BasedOnSelection),DisplayName=\"None\",InitialSpace=Loc ...(truncated)"
    Tag Controls=true
    Zero Controls=true
   )
   > Control Stack at Item_9 = RigVMFunctionReferenceNode(
    Add Bottom Null=False
    Add Null=true
    Add Secondary Control=False
    Control Suffix=_ctrl
    Item=(Name="dyn_BartClavicle_r",Type=Bone)
    Null Suffix=_null
    Parent=(Name="spine_05",Type=Bone)
    Secondary Settings="(Shape=(bVisible=True,Name=\"Default\",Color=(R=1,G=0,B=0,A=1),Transform=(Rotation=(X=0,Y=0,Z=0,W=1),Translation=(X=0,Y=0,Z=0),Scale3D=(X=1,Y=1,Z=1))),Proxy=(bIsProxy=False,ShapeVisibility=BasedOnSelection),DisplayName=\"None\",InitialSpace=LocalSpace,bUsePref ...(truncated)"
    Settings="(Shape=(bVisible=True,Name=\"Circle_Thick\",Color=(R=1,G=0,B=0,A=0),Transform=(Rotation=(X=0,Y=-0.7071,Z=0,W=0.7071),Translation=(X=0,Y=0,Z=0),Scale3D=(X=1,Y=1,Z=1))),Proxy=(bIsProxy=False,ShapeVisibility=BasedOnSelection),DisplayName=\"None\",InitialSpace=Loc ...(truncated)"
    Tag Controls=true
    Zero Controls=true
   )
   > Control Stack at Item_5 = RigVMFunctionReferenceNode(
    Add Bottom Null=False
    Add Null=true
    Add Secondary Control=False
    Control Suffix=_ctrl
    Item=(Name="dyn_BartHand_r",Type=Bone)
    Null Suffix=_null
    Parent=(Name="spine_05",Type=Bone)
    Secondary Settings="(Shape=(bVisible=True,Name=\"Default\",Color=(R=1,G=0,B=0,A=1),Transform=(Rotation=(X=0,Y=0,Z=0,W=1),Translation=(X=0,Y=0,Z=0),Scale3D=(X=1,Y=1,Z=1))),Proxy=(bIsProxy=False,ShapeVisibility=BasedOnSelection),DisplayName=\"None\",InitialSpace=LocalSpace,bUsePref ...(truncated)"
    Settings="(Shape=(bVisible=True,Name=\"Circle_Thick\",Color=(R=1,G=0,B=0,A=0),Transform=(Rotation=(X=0,Y=-0.7071,Z=0,W=0.7071),Translation=(X=0,Y=0,Z=0),Scale3D=(X=1,Y=1,Z=1))),Proxy=(bIsProxy=False,ShapeVisibility=BasedOnSelection),DisplayName=\"None\",InitialSpace=Loc ...(truncated)"
    Tag Controls=true
    Zero Controls=true
   )
   > Control Stack at Item_6 = RigVMFunctionReferenceNode(
    Add Bottom Null=False
    Add Null=true
    Add Secondary Control=False
    Control Suffix=_ctrl
    Item=(Name="dyn_BartLowerarm_r",Type=Bone)
    Null Suffix=_null
    Parent=(Name="spine_05",Type=Bone)
    Secondary Settings="(Shape=(bVisible=True,Name=\"Default\",Color=(R=1,G=0,B=0,A=1),Transform=(Rotation=(X=0,Y=0,Z=0,W=1),Translation=(X=0,Y=0,Z=0),Scale3D=(X=1,Y=1,Z=1))),Proxy=(bIsProxy=False,ShapeVisibility=BasedOnSelection),DisplayName=\"None\",InitialSpace=LocalSpace,bUsePref ...(truncated)"
    Settings="(Shape=(bVisible=True,Name=\"Circle_Thick\",Color=(R=1,G=0,B=0,A=0),Transform=(Rotation=(X=0,Y=-0.7071,Z=0,W=0.7071),Translation=(X=0,Y=0,Z=0),Scale3D=(X=1,Y=1,Z=1))),Proxy=(bIsProxy=False,ShapeVisibility=BasedOnSelection),DisplayName=\"None\",InitialSpace=Loc ...(truncated)"
    Tag Controls=true
    Zero Controls=true
   )
   > Control Stack at Item_7 = RigVMFunctionReferenceNode(
    Add Bottom Null=False
    Add Null=true
    Add Secondary Control=False
    Control Suffix=_ctrl
    Item=(Name="dyn_BartUpperarm_r",Type=Bone)
    Null Suffix=_null
    Parent=(Name="spine_05",Type=Bone)
    Secondary Settings="(Shape=(bVisible=True,Name=\"Default\",Color=(R=1,G=0,B=0,A=1),Transform=(Rotation=(X=0,Y=0,Z=0,W=1),Translation=(X=0,Y=0,Z=0),Scale3D=(X=1,Y=1,Z=1))),Proxy=(bIsProxy=False,ShapeVisibility=BasedOnSelection),DisplayName=\"None\",InitialSpace=LocalSpace,bUsePref ...(truncated)"
    Settings="(Shape=(bVisible=True,Name=\"Circle_Thick\",Color=(R=1,G=0,B=0,A=0),Transform=(Rotation=(X=0,Y=-0.7071,Z=0,W=0.7071),Translation=(X=0,Y=0,Z=0),Scale3D=(X=1,Y=1,Z=1))),Proxy=(bIsProxy=False,ShapeVisibility=BasedOnSelection),DisplayName=\"None\",InitialSpace=Loc ...(truncated)"
    Tag Controls=true
    Zero Controls=true
   )
   > Control Stack at Item_10 = RigVMFunctionReferenceNode(
    Add Bottom Null=False
    Add Null=False
    Add Secondary Control=False
    Control Suffix=_extra_ctrl
    Item=(Name="head",Type=Bone)
    Null Suffix=_extra_null
    Parent=(Name="spine_05",Type=Bone)
    Secondary Settings="(Shape=(bVisible=True,Name=\"Default\",Color=(R=1,G=0,B=0,A=1),Transform=(Rotation=(X=0,Y=0,Z=0,W=1),Translation=(X=0,Y=0,Z=0),Scale3D=(X=1,Y=1,Z=1))),Proxy=(bIsProxy=False,ShapeVisibility=BasedOnSelection),DisplayName=\"None\",InitialSpace=LocalSpace,bUsePref ...(truncated)"
    Settings="(Shape=(bVisible=True,Name=\"Default\",Color=(R=0,G=0,B=1,A=1),Transform=(Rotation=(X=0,Y=0,Z=0,W=1),Translation=(X=0,Y=0,Z=0),Scale3D=(X=1,Y=1,Z=1))),Proxy=(bIsProxy=False,ShapeVisibility=BasedOnSelection),DisplayName=\"None\",InitialSpace=LocalSpace,bUsePref ...(truncated)"
    Tag Controls=true
    Zero Controls=true
   )
   > RigVMFunctionReturnNode()
  ]}
 New Function_ContainedGraph {nodes [
   RigVMFunctionEntryNode()
   n1 = /Script/ControlRigSpline.RigUnit_ControlRigSplineFromPoints(
    Compression=0
    Points=((X=0,Y=0,Z=0),(X=0,Y=0,Z=0),(X=0,Y=0,Z=0),(X=0,Y=0,Z=0))
    SamplesPerSegment=4
    SplineMode=Hermite
    Stretch=0
    bClosed=false
    Points.1=&RigUnit_GetTransform(Item=(Type=Control,Name="dyn_BartUpperarm_l_ctrl"), bInitial=false).Transform.Translation
    Points.2=&RigUnit_GetTransform(Item=(Type=Control,Name="dyn_BartLowerarm_l_ctrl"), bInitial=False).Transform.Translation
    Points.3=&RigUnit_GetTransform(Item=(Type=Control,Name="dyn_BartHand_l_ctrl"), bInitial=False).Transform.Translation
    Points.0=&RigUnit_GetTransform(Item=(Type=Control,Name="dyn_BartClavicle_l_ctrl"), bInitial=false).Transform.Translation
   ) {is=struct RigVMUnitNode}
   > /Script/ControlRigSpline.RigUnit_DrawControlRigSpline(Color=(R=1,G=0,B=0,A=1), Detail=16, Thickness=1, Spline=@n1.Spline) {is=struct RigVMUnitNode}
   ItemArray = RigVMDispatchNode(Value=((Type=Bone,Name="dyn_BartClavicle_l"),(Type=Bone,Name="dyn_BartUpperarm_l"),(Type=Bone,Name="dyn_BartLowerarm_l"),(Type=Bone,Name="dyn_BartHand_l"))) {template="DISPATCH_RigVMDispatch_Constant(io Value)"}
   > /Script/ControlRigSpline.RigUnit_FitChainToSplineCurveItemArray(
    Alignment=Front
    DebugSettings=(bEnabled=False,Scale=1,CurveColor=(R=1,G=1,B=0,A=1),SegmentsColor=(R=1,G=0,B=0,A=1),WorldOffset=(Rotation=(X=0,Y=0,Z=0,W=1),Translation=(X=0,Y=0,Z=0),Scale3D=(X=1,Y=1,Z=1)))
    Maximum=1
    Minimum=0
    PoleVectorPosition=(X=0,Y=0,Z=0)
    PrimaryAxis=(X=1,Y=0,Z=0)
    RotationEaseType=Linear
    Rotations=((Rotation=(X=0,Y=0,Z=0,W=1),Ratio=0))
    SamplingPrecision=12
    SecondaryAxis=(X=0,Y=0,Z=0)
    Weight=1
    bPropagateToChildren=true
    Items=@ItemArray.Value
    Spline=@n1.Spline
   ) {is=struct RigVMUnitNode}
   > &RigUnit_ParentConstraint(
    AdvancedSettings=(InterpolationType=Average,RotationOrderForFilter=XZY)
    Child=(Type=Bone,Name="dyn_BartHand_l")
    Filter=(TranslationFilter=(bX=false,bY=false,bZ=false),RotationFilter=(bX=True,bY=True,bZ=True),ScaleFilter=(bX=false,bY=false,bZ=false))
    Parents=((Item=(Type=Control,Name="dyn_BartHand_l_ctrl"),Weight=1))
    Weight=1
   )
   n2 = &RigUnit_ControlRigSplineFromPoints(
    Compression=0
    Points=((X=0,Y=0,Z=0),(X=0,Y=0,Z=0),(X=0,Y=0,Z=0),(X=0,Y=0,Z=0))
    SamplesPerSegment=4
    SplineMode=Hermite
    Stretch=0
    bClosed=false
    Points.0=&RigUnit_GetTransform(Item=(Type=Control,Name="dyn_BartClavicle_r_ctrl"), bInitial=false).Transform.Translation
    Points.1=&RigUnit_GetTransform(Item=(Type=Control,Name="dyn_BartUpperarm_r_ctrl"), bInitial=false).Transform.Translation
    Points.2=&RigUnit_GetTransform(Item=(Type=Control,Name="dyn_BartLowerarm_r_ctrl"), bInitial=False).Transform.Translation
    Points.3=&RigUnit_GetTransform(Item=(Type=Control,Name="dyn_BartHand_r_ctrl"), bInitial=False).Transform.Translation
   )
   > &RigUnit_DrawControlRigSpline(Color=(R=1,G=0,B=0,A=1), Detail=16, Thickness=1, Spline=@n2.Spline)
   ItemArray_1 = RigVMDispatchNode(Value=((Type=Bone,Name="dyn_BartClavicle_r"),(Type=Bone,Name="dyn_BartUpperarm_r"),(Type=Bone,Name="dyn_BartLowerarm_r"),(Type=Bone,Name="dyn_BartHand_r"))) {template="DISPATCH_RigVMDispatch_Constant(io Value)"}
   > &RigUnit_FitChainToSplineCurveItemArray(
    Alignment=Front
    DebugSettings=(bEnabled=False,Scale=1,CurveColor=(R=1,G=1,B=0,A=1),SegmentsColor=(R=1,G=0,B=0,A=1),WorldOffset=(Rotation=(X=0,Y=0,Z=0,W=1),Translation=(X=0,Y=0,Z=0),Scale3D=(X=1,Y=1,Z=1)))
    Maximum=1
    Minimum=0
    PoleVectorPosition=(X=0,Y=0,Z=0)
    PrimaryAxis=(X=1,Y=0,Z=0)
    RotationEaseType=Linear
    Rotations=()
    SamplingPrecision=12
    SecondaryAxis=(X=0,Y=0,Z=0)
    Weight=1
    bPropagateToChildren=true
    Items=@ItemArray_1.Value
    Spline=@n2.Spline
   )
   > &RigUnit_ParentConstraint(
    AdvancedSettings=(InterpolationType=Average,RotationOrderForFilter=XZY)
    Child=(Type=Bone,Name="dyn_BartHand_r")
    Filter=(TranslationFilter=(bX=false,bY=false,bZ=false),RotationFilter=(bX=True,bY=True,bZ=True),ScaleFilter=(bX=false,bY=false,bZ=false))
    Parents=((Item=(Type=Control,Name="dyn_BartHand_r_ctrl"),Weight=1))
    Weight=1
   )
   > RigVMFunctionReturnNode()
  ]}
 New Function_ContainedGraph {nodes [
   RigVMFunctionEntryNode()
   n1 = &RigVMFunction_AlphaInterp(
    InterpSpeedDecreasing=10
    InterpSpeedIncreasing=10
    Scale=1
    bClampResult=False
    Value=&RigVMFunction_MathBoolToFloat(Value=RigVMVariableNode() {variable=isAiming}.Value).Result
   )
   n2 = &RigVMFunction_AlphaInterp(
    InterpSpeedDecreasing=10
    InterpSpeedIncreasing=10
    Scale=1
    bClampResult=False
    Value=&RigVMFunction_MathDoubleMax(
     B=RigVMDispatchNode(BoneName=root, Default=0, Name=deformRig_l_upperarm_fr_alpha) {template="DISPATCH_RigDispatch_GetAnimAttribute(in Name,in BoneName,in Default,out Value,out Found)"}.Value
     A=RigVMVariableNode() {variable=Char Move Right}.Value
    ).Result
   )
   n3 = &RigVMFunction_AlphaInterp(
    InterpSpeedDecreasing=10
    InterpSpeedIncreasing=10
    Scale=1
    bClampResult=False
    Value=&RigVMFunction_MathDoubleMax(
     B=RigVMDispatchNode(BoneName=root, Default=0, Name=deformRig_l_upperarm_fr_alpha) {template="DISPATCH_RigDispatch_GetAnimAttribute(in Name,in BoneName,in Default,out Value,out Found)"}.Value
     A=RigVMVariableNode() {variable=Char Move Left}.Value
    ).Result
   )
   n4 = &RigVMFunction_AlphaInterp(
    InterpSpeedDecreasing=10
    InterpSpeedIncreasing=10
    Scale=1
    bClampResult=False
    Value=&RigVMFunction_MathDoubleMax(
     B=RigVMDispatchNode(BoneName=root, Default=0, Name=deformRig_l_upperarm_fr_alpha) {template="DISPATCH_RigDispatch_GetAnimAttribute(in Name,in BoneName,in Default,out Value,out Found)"}.Value
     A=RigVMVariableNode() {variable=Char Backward}.Value
    ).Result
   )
   Maximum = &RigVMFunction_MathDoubleMax(
    B=RigVMDispatchNode(BoneName=root, Default=0, Name=deformRig_l_upperarm_fr_alpha) {template="DISPATCH_RigDispatch_GetAnimAttribute(in Name,in BoneName,in Default,out Value,out Found)"}.Value
    A=RigVMVariableNode() {variable=Char Forward}.Value
   )
   n5 = &RigVMFunction_AlphaInterp(InterpSpeedDecreasing=10, InterpSpeedIncreasing=10, Scale=1, bClampResult=False, bMapRange=false, Value=@Maximum.Result)
   Interpolate = /Script/RigVM.RigVMFunction_MathTransformLerp(
    A=(Rotation=(X=0,Y=0,Z=0,W=1),Translation=(X=0,Y=0,Z=0),Scale3D=(X=1,Y=1,Z=1))
    B=(Rotation=(X=0,Y=0,Z=0,W=1),Translation=(X=7.4,Y=2.4,Z=2.2),Scale3D=(X=1,Y=1,Z=1))
    T=@n5.Result
   ) {is=struct RigVMUnitNode, template="Interpolate::Execute(in A,in B,in T,out Result)"}
   n6 = &RigVMFunction_MathTransformLerp(
    B=(Rotation=(X=0,Y=0,Z=0,W=1),Translation=(X=-0.11,Y=-0.2,Z=0),Scale3D=(X=1,Y=1,Z=1))
    T=@n3.Result
    A=&RigVMFunction_MathTransformLerp(B=(Rotation=(X=0,Y=0,Z=0,W=1),Translation=(X=0.8,Y=-2.1,Z=-2.9),Scale3D=(X=1,Y=1,Z=1)), T=@n4.Result, A=@Interpolate.Result) {template="Interpolate::Execute(in A,in B,in T,out Result)"}.Result
   ) {template="Interpolate::Execute(in A,in B,in T,out Result)"}
   n7 = &RigVMFunction_MathTransformLerp(
    A=(Rotation=(X=0,Y=0,Z=0,W=1),Translation=(X=0,Y=0,Z=0),Scale3D=(X=1,Y=1,Z=1))
    B=(Rotation=(X=0,Y=0.1495,Z=0,W=0.9888),Translation=(X=5.6,Y=0,Z=2.1),Scale3D=(X=1,Y=1,Z=1))
    T=@n5.Result
   ) {template="Interpolate::Execute(in A,in B,in T,out Result)"}
   n8 = &RigVMFunction_MathTransformLerp(
    B=(Rotation=(X=0.1502,Y=-0.262,Z=-0.2505,W=0.9198),Translation=(X=0.85,Y=-3.5,Z=0.2),Scale3D=(X=1,Y=1,Z=1))
    T=@n3.Result
    A=&RigVMFunction_MathTransformLerp(B=(Rotation=(X=0.152,Y=-0.4827,Z=-0.0237,W=0.8622),Translation=(X=-4,Y=1.4,Z=4.3),Scale3D=(X=1,Y=1,Z=1)), T=@n4.Result, A=@n7.Result) {template="Interpolate::Execute(in A,in B,in T,out Result)"}.Result
   ) {template="Interpolate::Execute(in A,in B,in T,out Result)"}
   > &RigUnit_ModifyTransforms(
    ItemToModify="((Item=(Type=Control,Name=\"dyn_BartLowerarm_l_ctrl\"),Transform=(Rotation=(X=0,Y=0,Z=0,W=-1),Translation=(X=0,Y=0,Z=0),Scale3D=(X=1,Y=1,Z=1))),(Item=(Type=Control,Name=\"dyn_BartHand_l_ctrl\"),Transform=(Rotation=(X=0,Y=0,Z=0,W=1),Translation=(X=0,Y=0,Z=0),Sc ...(truncated)"
    Mode=OverrideLocal
    Weight=1
    WeightMinimum=0
    ItemToModify.0.Transform=&RigVMFunction_MathTransformLerp(
     B=(Rotation=(X=0,Y=0,Z=0,W=1),Translation=(X=-2,Y=2,Z=-2.5),Scale3D=(X=1,Y=1,Z=1))
     T=@n1.Result
     A=&RigVMFunction_MathTransformLerp(B=(Rotation=(X=0,Y=0,Z=0,W=1),Translation=(X=1.9,Y=2.3,Z=3.2),Scale3D=(X=1,Y=1,Z=1)), T=@n2.Result, A=@n6.Result) {template="Interpolate::Execute(in A,in B,in T,out Result)"}.Result
    ) {template="Interpolate::Execute(in A,in B,in T,out Result)"}.Result
    ItemToModify.1.Transform=&RigVMFunction_MathTransformLerp(
     B=(Rotation=(X=0.483,Y=-0.2241,Z=0.1294,W=0.8365),Translation=(X=-8,Y=2.5,Z=5),Scale3D=(X=1,Y=1,Z=1))
     T=@n1.Result
     A=&RigVMFunction_MathTransformLerp(B=(Rotation=(X=-0.2028,Y=-0.1572,Z=0.0118,W=0.9664),Translation=(X=5.6,Y=7.6,Z=3.2),Scale3D=(X=1,Y=1,Z=1)), T=@n2.Result, A=@n8.Result) {template="Interpolate::Execute(in A,in B,in T,out Result)"}.Result
    ) {template="Interpolate::Execute(in A,in B,in T,out Result)"}.Result
   )
   RigVMFunction_MathBoolToFloat = &RigVMFunction_MathBoolToFloat(Value=RigVMVariableNode() {variable=isAiming}.Value)
   n9 = &RigVMFunction_AlphaInterp(
    InterpSpeedDecreasing=10
    InterpSpeedIncreasing=10
    Scale=1
    bClampResult=False
    bMapRange=false
    Value=@RigVMFunction_MathBoolToFloat.Result
   )
   n10 = &RigVMFunction_AlphaInterp(
    InterpSpeedDecreasing=10
    InterpSpeedIncreasing=10
    Scale=1
    bClampResult=False
    Value=&RigVMFunction_MathDoubleMax(
     B=RigVMDispatchNode(BoneName=root, Default=0, Name=deformRig_r_upperarm_fr_alpha) {template="DISPATCH_RigDispatch_GetAnimAttribute(in Name,in BoneName,in Default,out Value,out Found)"}.Value
     A=RigVMVariableNode() {variable=Char Move Right}.Value
    ).Result
   )
   n11 = &RigVMFunction_AlphaInterp(
    InterpSpeedDecreasing=10
    InterpSpeedIncreasing=10
    Scale=1
    bClampResult=False
    Value=&RigVMFunction_MathDoubleMax(
     B=RigVMDispatchNode(BoneName=root, Default=0, Name=deformRig_r_upperarm_fr_alpha) {template="DISPATCH_RigDispatch_GetAnimAttribute(in Name,in BoneName,in Default,out Value,out Found)"}.Value
     A=RigVMVariableNode() {variable=Char Move Left}.Value
    ).Result
   )
   n12 = &RigVMFunction_AlphaInterp(
    InterpSpeedDecreasing=10
    InterpSpeedIncreasing=10
    Scale=1
    bClampResult=False
    Value=&RigVMFunction_MathDoubleMax(
     B=RigVMDispatchNode(BoneName=root, Default=0, Name=deformRig_r_upperarm_fr_alpha) {template="DISPATCH_RigDispatch_GetAnimAttribute(in Name,in BoneName,in Default,out Value,out Found)"}.Value
     A=RigVMVariableNode() {variable=Char Backward}.Value
    ).Result
   )
   n13 = &RigVMFunction_AlphaInterp(
    InterpSpeedDecreasing=10
    InterpSpeedIncreasing=10
    Scale=1
    bClampResult=False
    bMapRange=false
    Value=&RigVMFunction_MathDoubleMax(
     B=RigVMDispatchNode(BoneName=root, Default=0, Name=deformRig_r_upperarm_fr_alpha) {template="DISPATCH_RigDispatch_GetAnimAttribute(in Name,in BoneName,in Default,out Value,out Found)"}.Value
     A=RigVMVariableNode() {variable=Char Forward}.Value
    ).Result
   )
   n14 = &RigVMFunction_MathTransformLerp(
    A=(Rotation=(X=0,Y=0,Z=0,W=1),Translation=(X=0,Y=0,Z=0),Scale3D=(X=1,Y=1,Z=1))
    B=(Rotation=(X=0,Y=0.1495,Z=0,W=0.9888),Translation=(X=5.6,Y=0,Z=2.1),Scale3D=(X=1,Y=1,Z=1))
    T=@n13.Result
   ) {template="Interpolate::Execute(in A,in B,in T,out Result)"}
   n15 = &RigVMFunction_MathTransformLerp(
    B=(Rotation=(X=0.2028,Y=-0.1572,Z=-0.0118,W=0.9664),Translation=(X=5.6,Y=-7.6,Z=3.2),Scale3D=(X=1,Y=1,Z=1))
    T=@n11.Result
    A=&RigVMFunction_MathTransformLerp(B=(Rotation=(X=-0.152,Y=-0.4827,Z=0.0237,W=0.8622),Translation=(X=-4,Y=-1.4,Z=4.3),Scale3D=(X=1,Y=1,Z=1)), T=@n12.Result, A=@n14.Result) {template="Interpolate::Execute(in A,in B,in T,out Result)"}.Result
   ) {template="Interpolate::Execute(in A,in B,in T,out Result)"}
   n16 = &RigVMFunction_MathTransformLerp(
    A=(Rotation=(X=0,Y=0,Z=0,W=1),Translation=(X=0,Y=0,Z=0),Scale3D=(X=1,Y=1,Z=1))
    B=(Rotation=(X=0,Y=0,Z=0,W=1),Translation=(X=7.3868,Y=-2.4,Z=2.2),Scale3D=(X=1,Y=1,Z=1))
    T=@n13.Result
   ) {template="Interpolate::Execute(in A,in B,in T,out Result)"}
   n17 = &RigVMFunction_MathTransformLerp(
    B=(Rotation=(X=0,Y=0,Z=0,W=1),Translation=(X=1.9,Y=-2.3,Z=3.2),Scale3D=(X=1,Y=1,Z=1))
    T=@n11.Result
    A=&RigVMFunction_MathTransformLerp(B=(Rotation=(X=0,Y=0,Z=0,W=1),Translation=(X=0.8,Y=2.1,Z=-2.9),Scale3D=(X=1,Y=1,Z=1)), T=@n12.Result, A=@n16.Result) {template="Interpolate::Execute(in A,in B,in T,out Result)"}.Result
   ) {template="Interpolate::Execute(in A,in B,in T,out Result)"}
   > &RigUnit_ModifyTransforms(
    ItemToModify="((Item=(Type=Control,Name=\"dyn_BartLowerarm_r_ctrl\"),Transform=(Rotation=(X=0,Y=0,Z=0,W=-1),Translation=(X=0,Y=0,Z=0),Scale3D=(X=1,Y=1,Z=1))),(Item=(Type=Control,Name=\"dyn_BartHand_r_ctrl\"),Transform=(Rotation=(X=0,Y=0,Z=0,W=1),Translation=(X=0,Y=0,Z=0),Sc ...(truncated)"
    Mode=OverrideLocal
    Weight=1
    WeightMinimum=0
    ItemToModify.1.Transform=&RigVMFunction_MathTransformLerp(
     B=(Rotation=(X=-0.5245,Y=-0.0915,Z=-0.3415,W=0.7745),Translation=(X=-5,Y=-12,Z=8),Scale3D=(X=1,Y=1,Z=1))
     T=@n9.Result
     A=&RigVMFunction_MathTransformLerp(B=(Rotation=(X=-0.1341,Y=-0.2576,Z=0.255,W=0.9223),Translation=(X=0.8,Y=3.5,Z=0.2),Scale3D=(X=1,Y=1,Z=1)), T=@n10.Result, A=@n15.Result) {template="Interpolate::Execute(in A,in B,in T,out Result)"}.Result
    ) {template="Interpolate::Execute(in A,in B,in T,out Result)"}.Result
    ItemToModify.0.Transform=&RigVMFunction_MathTransformLerp(
     B=(Rotation=(X=0,Y=0,Z=0,W=1),Translation=(X=2,Y=-2,Z=-2.5),Scale3D=(X=1,Y=1,Z=1))
     T=@n9.Result
     A=&RigVMFunction_MathTransformLerp(B=(Rotation=(X=0,Y=0,Z=0,W=1),Translation=(X=0.1,Y=0.2,Z=0),Scale3D=(X=1,Y=1,Z=1)), T=@n10.Result, A=@n17.Result) {template="Interpolate::Execute(in A,in B,in T,out Result)"}.Result
    ) {template="Interpolate::Execute(in A,in B,in T,out Result)"}.Result
   )
   > RigVMFunctionReturnNode()
  ]}
 New Function_ContainedGraph {nodes [
   RigVMFunctionEntryNode()
   > &RigUnit_ParentConstraint(
    AdvancedSettings=(RotationOrderForFilter=XZY)
    Child=(Type=Bone,Name="base_backPiece")
    Parents=((Item=(Type=Bone,Name="spine_01"),Weight=0.0208),(Item=(Type=Bone,Name="spine_02"),Weight=0.0934),(Item=(Type=Bone,Name="spine_03"),Weight=0.2748),(Item=(Type=Bone,Name="spine_04"),Weight=0.2874),(Item=(Type=Bone,Name="spine_05"),Weight=0.3235))
    Weight=1
   )
   > RigVMFunctionReturnNode()
  ]}
 New Function_ContainedGraph {nodes [
   RigVMFunctionEntryNode()
   > RigVMFunction_Sequence = RigVMAggregateNode()
   RigVMFunctionReturnNode(ExecuteContext=@RigVMFunction_Sequence.F)
   n1 = &RigVMFunction_MathQuaternionSwingTwist(
    TwistAxis=(X=1,Y=0,Z=0)
    Input=&RigUnit_GetRelativeTransformForItem(Child=(Type=Bone,Name="head"), Parent=(Type=Control,Name="head_extra_ctrl"), bChildInitial=False, bParentInitial=False).RelativeTransform.Rotation
   )
   RigVMVariableNode(Value=&RigVMFunction_MathQuaternionToEuler(RotationOrder=ZYX, Value=@n1.Swing).Result, ExecuteContext=@RigVMFunction_Sequence.A) {variable=headSwing_rot}
   > RigVMVariableNode(Value=&RigVMFunction_MathQuaternionToEuler(RotationOrder=ZYX, Value=@n1.Twist).Result) {variable=headSwing_twist}
   n2 = &RigVMFunction_AnimEvalRichCurve(
    Curve="(EditorCurveData=(Keys=((InterpMode=RCIM_Cubic,TangentMode=RCTM_User,ArriveTangent=0.2796,LeaveTangent=0.2796),(InterpMode=RCIM_Cubic,TangentMode=RCTM_User,Time=0.5061,Value=0.359,ArriveTangent=1.2761,LeaveTangent=1.2761),(InterpMode=RCIM_Cubic,TangentMode ...(truncated)"
    SourceMaximum=-140
    SourceMinimum=0
    TargetMaximum=-44
    TargetMinimum=0
    Value=RigVMVariableNode() {variable=headSwing_rot}.Value.Z
   )
   n3 = &RigVMFunction_AlphaInterp(InterpSpeedDecreasing=10, InterpSpeedIncreasing=10, Scale=1, bClampResult=false, bInterpResult=false, bMapRange=true, Value=@n2.Result)
   &RigUnit_ModifyTransforms(
    ItemToModify=((Item=(Type=Bone,Name="base_mechNeckAreaAdjust"),Transform=(Rotation=(X=0,Y=0,Z=0,W=1),Translation=(X=0,Y=0,Z=0),Scale3D=(X=1,Y=1,Z=1))))
    Weight=1
    WeightMinimum=0
    ItemToModify.0.Transform.Rotation=&RigVMFunction_MathQuaternionFromEuler(Euler=(X=0,Y=0,Z=0), Euler.Y=&RigVMFunction_MathDoubleMul(B=12, A=@n3.Result).Result).Result
    ExecutePin=@RigVMFunction_Sequence.B
   )
   n4 = &RigVMFunction_AnimEvalRichCurve(
    Curve="(EditorCurveData=(Keys=((InterpMode=RCIM_Cubic,TangentMode=RCTM_User,ArriveTangent=0.2796,LeaveTangent=0.2796),(InterpMode=RCIM_Cubic,TangentMode=RCTM_User,Time=0.5061,Value=0.359,ArriveTangent=1.2761,LeaveTangent=1.2761),(InterpMode=RCIM_Cubic,TangentMode ...(truncated)"
    SourceMaximum=-140
    SourceMinimum=0
    TargetMaximum=-44
    TargetMinimum=0
    Value=RigVMVariableNode() {variable=headSwing_twist}.Value.X
   )
   n5 = &RigVMFunction_AlphaInterp(InterpSpeedDecreasing=10, InterpSpeedIncreasing=10, Scale=1, bClampResult=false, bInterpResult=false, bMapRange=true, Value=@n4.Result)
   n6 = &RigVMFunction_MathDoubleMul(B=3, A=@n5.Result)
   > &RigUnit_ModifyTransforms(
    ItemToModify=((Item=(Type=Bone,Name="base_mechNeckAreaAdjust"),Transform=(Rotation=(X=0,Y=0,Z=0,W=1),Translation=(X=0,Y=0,Z=0),Scale3D=(X=1,Y=1,Z=1))))
    Weight=1
    WeightMinimum=0
    ItemToModify.0.Transform.Rotation=&RigVMFunction_MathQuaternionFromEuler(Euler=(X=0,Y=0,Z=0), Euler.Y=&RigVMFunction_MathDoubleMul(B=3, A=@n5.Result).Result, Euler.X=@n6.Result, Euler.Z=@n6.Result).Result
   )
   n7 = &RigVMFunction_AnimEvalRichCurve(
    Curve="(EditorCurveData=(Keys=((InterpMode=RCIM_Cubic,TangentMode=RCTM_User,ArriveTangent=0.2796,LeaveTangent=0.2796),(InterpMode=RCIM_Cubic,TangentMode=RCTM_User,Time=0.5061,Value=0.359,ArriveTangent=1.2761,LeaveTangent=1.2761),(InterpMode=RCIM_Cubic,TangentMode ...(truncated)"
    SourceMaximum=140
    SourceMinimum=0
    TargetMaximum=44
    TargetMinimum=0
    Value=RigVMVariableNode() {variable=headSwing_twist}.Value.X
   )
   n8 = &RigVMFunction_AlphaInterp(InterpSpeedDecreasing=10, InterpSpeedIncreasing=10, Scale=1, bClampResult=false, bInterpResult=false, bMapRange=true, Value=@n7.Result)
   n9 = &RigVMFunction_MathDoubleMul(B=2, A=@n8.Result)
   > &RigUnit_ModifyTransforms(
    ItemToModify=((Item=(Type=Bone,Name="base_mechNeckAreaAdjust"),Transform=(Rotation=(X=0,Y=0,Z=0,W=1),Translation=(X=0,Y=0,Z=0),Scale3D=(X=1,Y=1,Z=1))))
    Weight=1
    WeightMinimum=0
    ItemToModify.0.Transform.Rotation=&RigVMFunction_MathQuaternionFromEuler(Euler=(X=0,Y=0,Z=0), Euler.X=@n9.Result, Euler.Z=@n9.Result, Euler.Y=&RigVMFunction_MathDoubleMul(B=-2.2, A=@n8.Result).Result).Result
   )
  ]}
 RigVMFunctionLibrary {nodes [
   Bar Setup = RigVMCollapseNode(ExecuteContext=())
   Barts Arms Spline Setup = RigVMCollapseNode(ExecuteContext=())
   Constraints = RigVMCollapseNode(ExecuteContext=())
   Controllers Setup = RigVMCollapseNode(ExecuteContext=())
   Legs Setup = RigVMCollapseNode(ExecuteContext=())
   Mech Joystick Movement Setup = RigVMCollapseNode(ExecuteContext=())
  ]}
 RigVMFunction_Sequence_1_ContainedGraph {nodes [
   n1 = RigVMFunctionEntryNode()
   RigVMFunction_Sequence = &RigVMFunction_Sequence(ExecuteContext=@n1.ExecuteContext)
   n2 = &RigVMFunction_Sequence(ExecuteContext=@RigVMFunction_Sequence.B)
   RigVMFunctionReturnNode(A=@RigVMFunction_Sequence.A, B=@n2.A, C=@n2.B)
  ]}
 RigVMFunction_Sequence_1_ContainedGraph {nodes [
   n1 = RigVMFunctionEntryNode()
   RigVMFunction_Sequence = &RigVMFunction_Sequence(ExecuteContext=@n1.ExecuteContext)
   n2 = &RigVMFunction_Sequence(ExecuteContext=@RigVMFunction_Sequence.B)
   RigVMFunctionReturnNode(A=@RigVMFunction_Sequence.A, B=@n2.A, F=@n2.B)
  ]}
 RigVMModel {nodes [
   n1 = &RigUnit_PrepareForExecution()
   &RigUnit_HierarchyImportFromSkeleton(NameSpace=None, bIncludeCurves=False, bIncludeMeshSockets=False, bIncludeVirtualBones=True, ExecutePin=@n1.ExecutePin)
   > Controllers Setup = RigVMFunctionReferenceNode()
   > Constraints = RigVMFunctionReferenceNode()
   RigUnit_BeginExecution = &RigUnit_BeginExecution()
   FN_DeformRig_FNC = RigVMFunctionReferenceNode(
    DeformRig_ Curve="(deformRig_l_upperarm_up_alpha_72_#1=true,deformRig_r_upperarm_up_alpha_73_#2=true,deformRig_l_upperarm_fr_alpha_74_#3=true,deformRig_r_upperarm_fr_alpha_75_CB51B80C4 ...(truncated)"
    DeformRig_Bool="(upperarm_1_#4=True,elbow_3_#5=True,wrist_7_#6=True,groin_8_#7=True,glute_17_#8=True,knee_9_40C408324CA261 ...(truncated)"
    DeformRig_Debug=(upperarm_debug_18_#4=False,wrist_debug_20_#6=False,lat_debug_26_#9=False,groinGlute_debug_27_#7=False)
    DeformRig_Value="(upperarm_up_18_#10=(Rotation=(X=0.0,Y=-0.0,Z=0.0,W=1.0),Translation=(X=0.0,Y=0.0,Z=8.0),Scale3D=(X=0.70,Y=1.0,Z=1.0)),upperarm_fwd_fr_19_#11=(Rotation=(X=0.0,Y=-0.0,Z=0.0,W=1.0),Translation=(X=0.0, ...(truncated)"
    ExecuteContext=@RigUnit_BeginExecution.ExecutePin
   )
   > Barts Arms Spline Setup = RigVMFunctionReferenceNode()
   > Legs Setup = RigVMFunctionReferenceNode()
   > Mech Joystick Movement Setup = RigVMFunctionReferenceNode()
   > Bar Setup = RigVMFunctionReferenceNode()
   addFKCtrl = RigVMFunctionReferenceNode(
    Custom_FKCtrl=()
    ExecuteContext=()
    Options_FKCtrl="(pelvis_1_#12=true,spine_3_#13=true,clavicle_5_#14=true,arm_7_#15=false,leg_9_#16=true,head_11_46DC15EF465CD297 ...(truncated)"
    Shape_Settings="(pelvis_2_#17=(bVisible=True,Name=\"Hexagon_Thin\",Color=(R=0.0117,G=0.8281,B=1,A=1),Transform=(Rotation=(X=0,Y=-0.7071,Z=0,W=0.7071),Translation=(X=2,Y=0,Z=0),Scale3D=(X=5,Y=5,Z=7))),spine_4_#18=(bVi ...(truncated)"
   )
   setFKCtrl = RigVMFunctionReferenceNode(ExecuteContext=(), setFK_Ctrl=false)
  ]}
]
variables [
 BartLegSwing_L {type=FVector, default=(X=0,Y=0,Z=0)}
 BartLegSwing_R {type=FVector, default=(X=0,Y=0,Z=0)}
 Char Backward {type=double, default=0}
 Char Forward {type=double, default=0}
 Char Move Left {type=double, default=0}
 Char Move Right {type=double, default=0}
 headSwing_rot {type=FVector, default=(X=0,Y=0,Z=0)}
 headSwing_twist {type=FVector, default=(X=0,Y=0,Z=0)}
 isAiming {type=bool, default=False}
 isCrouching {type=bool, default=False}
 rifleEquipped {type=double, default=0}
]
