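# Physics asset for the large male QueenTruth Rind player body, covering the main skeleton and an attached "Bart" pelvis-and-legs rig.
# Capsules cover torso, arms, hands, thighs and calves, with extra shapes on spine_03 and thigh_r.
# BartPelvis hangs off spine_03 with narrow limits and drives two three-link dyn leg chains through tiny base bodies.
# The chains have tight swing and zero-twist limits, while main-body and leg-base joints rotate freely.
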PhysicsAsset /BRCosmetics/Characters/Player/Male/Large/Bodies/M_LRG_QueenTruth_Rind/Meshes/M_LRG_QueenTruth_Rind_Physics
bodies {
 base_BartLeg_l [{shapeType=Capsule, length=0, radius=0.505, shapeName=base_BartLeg_l_capsule}]
 base_BartLeg_r [{shapeType=Capsule, length=0, radius=0.505, shapeName=base_BartLeg_r_capsule}]
 base_BartPelvis [{shapeType=Capsule, length=9.9408, radius=8.0592, shapeName=base_BartPelvis_capsule, center=<3.2167 0 6.5195>, rotation=<rot 0 0 90>}]
 calf_l [{shapeType=Capsule, length=30, radius=8, shapeName=calf_l_capsule, center=<-22.2744 0 0>, rotation=<rot 90 0 0>}]
 calf_r [{shapeType=Capsule, length=30, radius=8, shapeName=calf_r_capsule, center=<22.2744 0 0>, rotation=<rot 90 180 0>}]
 dyn_BartLeg_l_01 [{shapeType=Capsule, length=4.0733, radius=3.2342, shapeName=dyn_BartLeg_l_01_capsule, center=<4.9823 0 0>, rotation=<rot 90 0 0>}]
 dyn_BartLeg_l_02 [{shapeType=Capsule, length=4.6211, radius=3.3706, shapeName=dyn_BartLeg_l_02_capsule, center=<4 0 0>, rotation=<rot 90 0 0>}]
 dyn_BartLeg_l_03 [{shapeType=Capsule, length=6.5, radius=6, shapeName=dyn_BartLeg_l_03_capsule, center=<7.5 0 3.5>, rotation=<rot -18.6869 0 0>}]
 dyn_BartLeg_r_01 [{shapeType=Capsule, length=4.0733, radius=3.2342, shapeName=dyn_BartLeg_r_01_capsule, center=<4.9823 0 0>, rotation=<rot 90 0 0>}]
 dyn_BartLeg_r_02 [{shapeType=Capsule, length=4.6211, radius=3.3706, shapeName=dyn_BartLeg_r_02_capsule, center=<4 0 0>, rotation=<rot 90 0 0>}]
 dyn_BartLeg_r_03 [{shapeType=Capsule, length=6.5, radius=6, shapeName=dyn_BartLeg_r_03_capsule, center=<7.5 0 3.5>, rotation=<rot -18.6869 0 0>}]
 hand_l [{shapeType=Capsule, length=5.7801, radius=6.5, shapeName=hand_l_capsule, center=<10.0328 -3.2708 0>, rotation=<rot 90 0 0>}]
 hand_r [{shapeType=Capsule, length=5.7801, radius=6.5, shapeName=hand_r_capsule, center=<-10.0328 3.2708 0>, rotation=<rot 90 180 0>}]
 lowerarm_l [{shapeType=Capsule, length=19.4088, radius=6.0912, shapeName=lowerarm_l_capsule, center=<16.4431 -2.6419 6.1594>, rotation=<rot 90 0 0>}]
 lowerarm_r [{shapeType=Capsule, length=19.4088, radius=6.0912, shapeName=lowerarm_r_capsule, center=<-16.4431 0 0>, rotation=<rot 90 180 0>}]
 pelvis [{shapeType=Capsule, length=20.8214, radius=8.36, shapeName=pelvis_capsule}]
 spine_01 [{shapeType=Capsule, length=16.9989, radius=6.5515, shapeName=spine_01_capsule}]
 spine_02 [{shapeType=Capsule, length=19.3016, radius=6.4963, shapeName=spine_02_capsule}]
 spine_03 [
  {shapeType=Capsule, length=21.748, radius=4.1338, shapeName=spine_03_capsule}
  {shapeType=Capsule, length=15, radius=4, shapeName=spine_03_capsule, center=<-5.8874 6.4758 0>, rotation=<rot 0 24.8237 90>}
  {shapeType=Capsule, length=15, radius=4, shapeName=spine_03_capsule, center=<-3.7958 18.8538 0>, rotation=<rot 90 33.8175 0>}
 ]
 thigh_l [{shapeType=Capsule, length=32.9979, radius=5.0021, shapeName=thigh_l_capsule, center=<-21.3736 0 0>, rotation=<rot 90 0 0>}]
 thigh_r [
  {shapeType=Capsule, length=32.9979, radius=5.0021, shapeName=thigh_r_capsule, center=<21.3736 0 0>, rotation=<rot 90 180 0>}
  {shapeType=Capsule, length=0, radius=7.2281, shapeName=thigh_r_capsule, center=<26.9805 -5.8735 0>}
 ]
 upperarm_l [{shapeType=Capsule, length=20, radius=6, shapeName=upperarm_l_capsule, center=<14.5141 0 0>, rotation=<rot 90 0 0>}]
 upperarm_r [{shapeType=Capsule, length=20, radius=6, shapeName=upperarm_r_capsule, center=<-14.5141 0 0>, rotation=<rot 90 180 0>}]
}
constraints [
 {
  bone1Name=spine_01
  bone2Name=pelvis
  swing1LimitDegrees=45
  swing1Motion=Free
  swing2LimitDegrees=45
  swing2Motion=Free
  twistLimitDegrees=45
  twistMotion=Free
 }
 {
  bone1Name=thigh_l
  bone2Name=pelvis
  swing1LimitDegrees=45
  swing1Motion=Free
  swing2LimitDegrees=45
  swing2Motion=Free
  twistLimitDegrees=45
  twistMotion=Free
 }
 {
  bone1Name=calf_l
  bone2Name=thigh_l
  swing1LimitDegrees=45
  swing1Motion=Free
  swing2LimitDegrees=45
  swing2Motion=Free
  twistLimitDegrees=45
  twistMotion=Free
 }
 {
  bone1Name=thigh_r
  bone2Name=pelvis
  swing1LimitDegrees=45
  swing1Motion=Free
  swing2LimitDegrees=45
  swing2Motion=Free
  twistLimitDegrees=45
  twistMotion=Free
 }
 {
  bone1Name=calf_r
  bone2Name=thigh_r
  swing1LimitDegrees=45
  swing1Motion=Free
  swing2LimitDegrees=45
  swing2Motion=Free
  twistLimitDegrees=45
  twistMotion=Free
 }
 {bone1Name=base_BartPelvis, bone2Name=spine_03, swing1LimitDegrees=5, swing2LimitDegrees=15, twistLimitDegrees=0}
 {bone1Name=dyn_BartLeg_l_01, bone2Name=base_BartLeg_l, swing1LimitDegrees=10, swing2LimitDegrees=25, twistLimitDegrees=0}
 {bone1Name=dyn_BartLeg_l_02, bone2Name=dyn_BartLeg_l_01, swing1LimitDegrees=15, swing2LimitDegrees=30, twistLimitDegrees=0}
 {bone1Name=dyn_BartLeg_l_03, bone2Name=dyn_BartLeg_l_02, swing1LimitDegrees=20, swing2LimitDegrees=30, twistLimitDegrees=0}
 {bone1Name=dyn_BartLeg_r_01, bone2Name=base_BartLeg_r, swing1LimitDegrees=10, swing2LimitDegrees=20, twistLimitDegrees=0}
 {bone1Name=dyn_BartLeg_r_02, bone2Name=dyn_BartLeg_r_01, swing1LimitDegrees=15, swing2LimitDegrees=25, twistLimitDegrees=0}
 {bone1Name=dyn_BartLeg_r_03, bone2Name=dyn_BartLeg_r_02, swing1LimitDegrees=20, swing2LimitDegrees=30, twistLimitDegrees=0}
 {
  bone1Name=spine_02
  bone2Name=spine_01
  swing1LimitDegrees=45
  swing1Motion=Free
  swing2LimitDegrees=45
  swing2Motion=Free
  twistLimitDegrees=45
  twistMotion=Free
 }
 {
  bone1Name=spine_03
  bone2Name=spine_02
  swing1LimitDegrees=45
  swing1Motion=Free
  swing2LimitDegrees=45
  swing2Motion=Free
  twistLimitDegrees=45
  twistMotion=Free
 }
 {
  bone1Name=upperarm_l
  bone2Name=spine_03
  swing1LimitDegrees=45
  swing1Motion=Free
  swing2LimitDegrees=45
  swing2Motion=Free
  twistLimitDegrees=45
  twistMotion=Free
 }
 {
  bone1Name=lowerarm_l
  bone2Name=upperarm_l
  swing1LimitDegrees=45
  swing1Motion=Free
  swing2LimitDegrees=45
  swing2Motion=Free
  twistLimitDegrees=45
  twistMotion=Free
 }
 {
  bone1Name=hand_l
  bone2Name=lowerarm_l
  swing1LimitDegrees=45
  swing1Motion=Free
  swing2LimitDegrees=45
  swing2Motion=Free
  twistLimitDegrees=45
  twistMotion=Free
 }
 {
  bone1Name=upperarm_r
  bone2Name=spine_03
  swing1LimitDegrees=45
  swing1Motion=Free
  swing2LimitDegrees=45
  swing2Motion=Free
  twistLimitDegrees=45
  twistMotion=Free
 }
 {
  bone1Name=lowerarm_r
  bone2Name=upperarm_r
  swing1LimitDegrees=45
  swing1Motion=Free
  swing2LimitDegrees=45
  swing2Motion=Free
  twistLimitDegrees=45
  twistMotion=Free
 }
 {
  bone1Name=hand_r
  bone2Name=lowerarm_r
  swing1LimitDegrees=45
  swing1Motion=Free
  swing2LimitDegrees=45
  swing2Motion=Free
  twistLimitDegrees=45
  twistMotion=Free
 }
 {
  bone1Name=base_BartLeg_l
  bone2Name=base_BartPelvis
  swing1LimitDegrees=45
  swing1Motion=Free
  swing2LimitDegrees=45
  swing2Motion=Free
  twistLimitDegrees=45
  twistMotion=Free
 }
 {
  bone1Name=base_BartLeg_r
  bone2Name=base_BartPelvis
  swing1LimitDegrees=45
  swing1Motion=Free
  swing2LimitDegrees=45
  swing2Motion=Free
  twistLimitDegrees=45
  twistMotion=Free
 }
]
