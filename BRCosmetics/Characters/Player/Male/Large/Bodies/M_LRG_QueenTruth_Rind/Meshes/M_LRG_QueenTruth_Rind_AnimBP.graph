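# Post-process style AnimBP for the QueenTruth Rind large male body cosmetic, layering accessory motion onto the pose copied from the attached parent mesh.
# A Control Rig takes movement, crouch, aim and rifle inputs, then bone constraints and additive offsets adjust shoulder wheels, handlebars, controllers and back piece before RigidBody simulation.
# The event graph derives movement values, spine-driven back-piece offsets, rifle/shotgun weight and state-based gravity and simulation settings.
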AnimBlueprint /BRCosmetics/Characters/Player/Male/Large/Bodies/M_LRG_QueenTruth_Rind/Meshes/M_LRG_QueenTruth_Rind_AnimBP
target_skeleton=&/M_LRG_QueenTruth_Rind_Skeleton
graphs [
 AnimGraph {nodes [
   n1 = Control Rig!(
    Source=Copy Pose From Mesh() {bUseAttachedParent=true}
    Alpha
    isCrouching:bool=Get bIsCrouching()
    Char Forward:double=Get Char Forward() -> (Char Forward:double)
    Char Backward:double=Get Char Backward() -> (Char Backward:double)
    Char Move Right:double=Get Char Move Right() -> (Char Move Right:double)
    Char Move Left:double=Get Char Move Left() -> (Char Move Left:double)
    isAiming:bool=Get bIsTargeting()
    rifleEquipped:double=Get isAssaultRifleEquipped float() -> (isAssaultRifleEquipped float:double)
   ) -> (Pose) {controlRigAssetReference {blueprintRigClass=&/M_LRG_QueenTruth_Rind_CtrlRig.M_LRG_QueenTruth_Rind_CtrlRig_C}}
   n2 = "Constraint - base_shoulderWheel_l (RR)"!(ComponentPose=Local To Component(LocalPose=@n1), Alpha=1, ConstraintWeights_0=0.55, ConstraintWeights_1:float=0.45) -> (Pose) {
    boneToModify {boneName=base_shoulderWheel_l}
    constraintSetup [{transformType=Rotation, targetBone {boneName=clavicle_l}}, {transformType=Rotation, targetBone {boneName=upperarm_l}}]
    constraintWeights [1,1]
   }
   n3 = "Constraint - base_shoulderWheel_r (RR)"!(ComponentPose=@n2, Alpha=1, ConstraintWeights_0=0.55, ConstraintWeights_1:float=0.45) -> (Pose) {
    boneToModify {boneName=base_shoulderWheel_r}
    constraintSetup [{transformType=Rotation, targetBone {boneName=clavicle_r}}, {transformType=Rotation, targetBone {boneName=upperarm_r}}]
    constraintWeights [1,1]
   }
   n4 = "Transform (Modify) Bone - Bone: base_handlebars_l"(ComponentPose=@n3, Alpha=1, Translation="0,0,0", Rotation="-52.226,-44.977,139.966", Scale="1,1,1") {
    rotationMode=BMM_Additive
    rotationSpace=BCS_BoneSpace
    translationMode=BMM_Additive
    translationSpace=BCS_BoneSpace
    alphaScaleBiasClamp {bInterpResult=true}
    boneToModify {boneName=base_handlebars_l}
    rotation=<rot -52.226 -44.977 139.966>
   }
   n5 = "Transform (Modify) Bone - Bone: base_handlebars_r"(ComponentPose=@n4, Alpha=1, Translation="0,0,0", Rotation="-65.887,33.972,-70.352", Scale="1,1,1") {
    rotationMode=BMM_Additive
    rotationSpace=BCS_BoneSpace
    translationMode=BMM_Additive
    translationSpace=BCS_BoneSpace
    alphaScaleBiasClamp {bInterpResult=true}
    boneToModify {boneName=base_handlebars_r}
    rotation=<rot -65.887 33.9721 -70.3521>
   }
   n6 = "Transform (Modify) Bone - Bone: base_handlebars_l"!(ComponentPose=@n5, bAlphaBoolEnabled:bool=True, Translation="0,0,0", Rotation="-10,30,15", Scale="1,1,1") -> (Pose) {
    alphaInputType=Bool
    rotationMode=BMM_Additive
    rotationSpace=BCS_BoneSpace
    translationMode=BMM_Additive
    translationSpace=BCS_BoneSpace
    alphaScaleBiasClamp {bInterpResult=true}
    boneToModify {boneName=base_handlebars_l}
    rotation=<rot -10 30 15>
   }
   n7 = "Transform (Modify) Bone - Bone: base_handlebars_r"!(ComponentPose=@n6, bAlphaBoolEnabled:bool=True, Translation="0,0,0", Rotation="-10,-30,-15", Scale="1,1,1") -> (Pose) {
    alphaInputType=Bool
    rotationMode=BMM_Additive
    rotationSpace=BCS_BoneSpace
    translationMode=BMM_Additive
    translationSpace=BCS_BoneSpace
    alphaScaleBiasClamp {bInterpResult=true}
    boneToModify {boneName=base_handlebars_r}
    rotation=<rot -10 -30 -15>
   }
   n8 = "Transform (Modify) Bone - Bone: base_shoulderWheel_l"!(ComponentPose=@n7, bAlphaBoolEnabled:bool=True, Translation="0,0,0", Rotation="-15,0,0", Scale="1,1,1") -> (Pose) {
    alphaInputType=Bool
    rotationMode=BMM_Additive
    rotationSpace=BCS_BoneSpace
    translationMode=BMM_Additive
    translationSpace=BCS_BoneSpace
    alphaScaleBiasClamp {bInterpResult=true}
    boneToModify {boneName=base_shoulderWheel_l}
    rotation {pitch=-15}
   }
   n9 = "Transform (Modify) Bone - Bone: base_shoulderWheel_r"!(ComponentPose=@n8, bAlphaBoolEnabled:bool=True, Translation="0,0,0", Rotation="-15,0,0", Scale="1,1,1") -> (Pose) {
    alphaInputType=Bool
    rotationMode=BMM_Additive
    rotationSpace=BCS_BoneSpace
    translationMode=BMM_Additive
    translationSpace=BCS_BoneSpace
    alphaScaleBiasClamp {bInterpResult=true}
    boneToModify {boneName=base_shoulderWheel_r}
    rotation {pitch=-15}
   }
   n10 = "Transform (Modify) Bone - Bone: dyn_backPiece"(
    ComponentPose=@n9
    Alpha=1
    Translation=Get BP Translate() -> (BP Translate:Vector)
    Rotation=Get BP Rotate() -> (BP Rotate:Rotator)
    Scale="1,1,1"
   ) {
    rotationMode=BMM_Additive
    rotationSpace=BCS_BoneSpace
    translationMode=BMM_Additive
    translationSpace=BCS_BoneSpace
    boneToModify {boneName=dyn_backPiece}
   }
   n11 = "Transform (Modify) Bone - Bone: base_shoulderWheel_l"(ComponentPose=@n10, bAlphaBoolEnabled=True, Translation="0,0,0", Rotation="20.005,-9.99,0", Scale="1,1,1") {
    alphaInputType=Bool
    rotationMode=BMM_Additive
    rotationSpace=BCS_BoneSpace
    translationMode=BMM_Additive
    translationSpace=BCS_BoneSpace
    alphaScaleBiasClamp {bInterpResult=true}
    boneToModify {boneName=base_shoulderWheel_l}
    rotation {pitch=20.005, yaw=-9.99}
   }
   n12 = "Transform (Modify) Bone - Bone: base_shoulderWheel_r"(ComponentPose=@n11, bAlphaBoolEnabled=True, Translation="0,0,0", Rotation="19.710,10.612,3.602", Scale="1,1,1") {
    alphaInputType=Bool
    rotationMode=BMM_Additive
    rotationSpace=BCS_BoneSpace
    translationMode=BMM_Additive
    translationSpace=BCS_BoneSpace
    alphaScaleBiasClamp {bInterpResult=true}
    boneToModify {boneName=base_shoulderWheel_r}
    rotation=<rot 19.71 10.612 3.602>
   }
   n13 = "Transform (Modify) Bone - Bone: base_controller_l"!(ComponentPose=@n12, bAlphaBoolEnabled:bool=True, Translation="-2,2,0", Rotation="0,30,0", Scale="1,1,1") -> (Pose) {
    alphaInputType=Bool
    rotationMode=BMM_Additive
    rotationSpace=BCS_BoneSpace
    translationMode=BMM_Additive
    translationSpace=BCS_BoneSpace
    alphaScaleBiasClamp {bInterpResult=true}
    boneToModify {boneName=base_controller_l}
    rotation {yaw=30}
    translation=<-2 2>
   }
   n14 = "Transform (Modify) Bone - Bone: base_controller_r"!(ComponentPose=@n13, bAlphaBoolEnabled:bool=True, Translation="2,-2,0", Rotation="0,50,0", Scale="1,1,1") -> (Pose) {
    alphaInputType=Bool
    rotationMode=BMM_Additive
    rotationSpace=BCS_BoneSpace
    translationMode=BMM_Additive
    translationSpace=BCS_BoneSpace
    alphaScaleBiasClamp {bInterpResult=true}
    boneToModify {boneName=base_controller_r}
    rotation {yaw=50}
    translation=<2 -2>
   }
   n15 = "Transform (Modify) Bone - Bone: base_controller_r"!(ComponentPose=@n14, Alpha:float=Get isAssaultRifleEquipped float(), Translation="-0.000,-4.761,-0.180", Rotation="0,0,0", Scale="1,1,1") -> (Pose) {
    rotationMode=BMM_Additive
    rotationSpace=BCS_BoneSpace
    translationMode=BMM_Additive
    translationSpace=BCS_BoneSpace
    alphaScaleBiasClamp {bInterpResult=true}
    boneToModify {boneName=base_controller_r}
    translation {y=-4.761, z=-0.18}
   }
   n16 = RigidBody!(
    ComponentPose=@n15
    Alpha
    OverrideWorldGravity:Vector="0,0,0"
    ExternalForce
    SimSpaceSettings:Sim Space Settings Structure=(WorldAlpha=0,VelocityScaleZ=1,DampingAlpha=1,MaxLinearVelocity=10000,MaxAngularVelocity=10000,MaxLinearAcceleration=10000,MaxAngularAcceleration=10000,ExternalLinearDragV=(X=0,Y=0,Z=0),ExternalLinearVelocity=(X=0,Y=0,Z=0),ExternalAngularVelocity=(X=0,Y=0,Z=0))
   ) -> (Pose) {
    bClampLinearTranslationLimitToRefPose=true
    bOverrideWorldGravity=true
    overridePhysicsAsset=&/M_LRG_QueenTruth_Rind_Physics
    simulationSpace=BaseBoneSpace
    baseBoneRef {boneName=spine_03}
   }
   Output Pose(Result=Component To Local(ComponentPose=@n16)) {layerGroup=None}
  ]}
 EventGraph {nodes [
   Event BlueprintUpdateAnimation()
   > Get!(execute:Exec) -> (then:Exec, else:Exec, SkelMesh:Skeletal Mesh Component Object Reference)
   > Get!(execute) -> (then, else, Pawn:Fort Player Pawn Object Reference)
   > n1 = Sequence!(execute) -> (then_0, then_1, then_2:Exec, then_3:Exec)
   n2 = PlayerMovement(execute=@n1.then_1, AnimInstance=Self-Reference(), XDotProductCoefficient=25, YDotProductCoefficient=25, ZDotProductCoefficient=1)
   > Set Char Move Right(Char Move Right:double=MapRangeClamped(Value=@n2, InRangeA=-1, InRangeB=25)) -> (Output_Get:double)
   > Set Char Move Left(Char Move Left:double=MapRangeClamped(Value=@n2, InRangeA=1, InRangeB=-25)) -> (Output_Get:double)
   > Set Char Forward(Char Forward:double=MapRangeClamped(Value=@n2.YMovement, InRangeA=-1, InRangeB=25)) -> (Output_Get:double)
   > Set Char Backward(Char Backward:double=MapRangeClamped(Value=@n2.YMovement, InRangeA=1, InRangeB=-25)) -> (Output_Get:double)
   n3 = "Gravity Override State Function (Nativized)"!(
    execute=@n1.then_0
    CharacterPartAnimInstance:Custom Character Part Anim Instance Object Reference=Self-Reference()
    GravityOverrideParameterSet:Fort Gravity Override Parameter Set Structure=Get Gravity Override Params() -> (Gravity Override Params:Fort Gravity Override Parameter Set Structure)
    RigidBodyAnimNodeParameterSet:Fort Rigid Body Anim Node Parameter Set Structure=Get RBAN Params() -> (RBAN Params:Fort Rigid Body Anim Node Parameter Set Structure)
    DebugDraw:bool=false
    GameTimeInSeconds:float=GetGameTimeInSeconds()
    DebugJointName:Name=root
    StateLogicLayerName:Name=none
   ) -> (
    then
    OutGravityOverride:Vector
    OutCurrentState:Name=None
    OutGravityOverrideParameters:Fort Gravity Override Parameters Structure
    OutRigidBodyAnimNodeParameters:Fort Rigid Body Anim Node Parameters Structure
   )
   > Set Gravity Override(Gravity Override=@n3)
   > Set RBAN Settings Out(RBAN Settings Out=@n3.OutRigidBodyAnimNodeParameters)
   n4 = GetOwnerSkeletalMeshComponent()
   n5 = BreakRotator(InRot=BreakTransform(InTransform=GetDeltaTransformFromRefPose(self=@n4, BoneName=spine_05, BaseName=spine_01)).Rotation)
   n6 = MapRangeClamped(Value=@n5.Yaw, InRangeA=0, InRangeB=-10, OutRangeB=10)
   Set BP Rotate(
    execute=@n1.then_3
    BP Rotate:Rotator=MakeRotator(Pitch=FInterpEaseInOut!(A:double=0, B:double=@n6, Alpha:double=0, Exponent:double=0) -> (ReturnValue:double))
   ) -> (Output_Get:Rotator)
   n7 = MapRangeClamped(Value=@n5.Yaw, InRangeA=0, InRangeB=-15, OutRangeB=5)
   > Set BP Translate(BP Translate:Vector=MakeVector(X=FInterpEaseInOut(A=0, B=@n7, Alpha=0, Exponent=0), Y=0)) -> (Output_Get:Vector)
   n8 = GetAlphaFromReferenceTransformDelta(SkeletalMeshComponent=Get SkelMesh(), BoneName=spine_05, BaseName=head, InputRange=(X=0,Y=110), TwistAxis="0,0,1")
   Set Head Rot(execute=@n1.then_2, Head Rot:float=@n8) -> (Output_Get:float)
   n9 = Get bIsTargeting()
   > Set ADS UP(ADS UP:float=float * float(A=@n8, B="To Float (Boolean)"(InBool=@n9)) {is=title K2Node_PromotableOperator}) -> (Output_Get:float)
   Event OnWeaponChanged()
   > Set isAssaultRifleEquipped float(isAssaultRifleEquipped float:double=0) -> (Output_Get:double)
   > n10 = Switch on EFortWeaponCoreAnimation!(execute, Selection:EFortWeaponCoreAnimation Enum=Get WeaponCoreAnim!() -> (WeaponCoreAnim)) -> (
    Melee:Exec
    Pistol:Exec
    Shotgun:Exec
    PaperBlueprint:Exec
    Rifle:Exec
    MeleeOneHand:Exec
    MachinePistol:Exec
    RocketLauncher:Exec
    GrenadeLauncher:Exec
    GoingCommando:Exec
    AssaultRifle:Exec
    TacticalShotgun:Exec
    SniperRifle:Exec
    TrapPlacement:Exec
    ShoulderLauncher:Exec
    AbilityDecoTool:Exec
    Crossbow:Exec
    C4:Exec
    RemoteControl:Exec
    DualWield:Exec
    AR_BullPup:Exec
    AR_ForwardGrip:Exec
    MedPackPaddles:Exec
    SMG_P90:Exec
    AR_DrumGun:Exec
    Consumable_Small:Exec
    Consumable_Large:Exec
    Balloon:Exec
    MountedTurret:Exec
    CreativeTool:Exec
    ExplosiveBow:Exec
    AshtonIndigo:Exec
    AshtonChicago:Exec
    MeleeDualWield:Exec
    Unarmed:Exec
   )
   Set isAssaultRifleEquipped float(execute=[@n10.AssaultRifle, @n10.Shotgun], isAssaultRifleEquipped float=1)
   Event BlueprintInitializeAnimation()
   > Set SkelMesh(SkelMesh=GetOwnerSkeletalMeshComponent())
   > n11 = Cast To FortPlayerPawn(Object=TryGetFortPawn())
   > Set Pawn(Pawn:Fort Player Pawn Object Reference=@n11) -> (Output_Get:Fort Player Pawn Object Reference)
  ]}
]
variables [
 ADS UP {type=number}
 BP Rotate {type=Rotator}
 BP Translate {type=Vector}
 Char Backward {type=number}
 Char Forward {type=number}
 Char Move Left {type=number}
 Char Move Right {type=number}
 Gravity Override {type=Vector}
 Gravity Override Params {
  type=FortGravityOverrideParameterSet
  default=(Skydive_GoingUp=(GravityOverride=(WindFrequency=(X=1.500000,Y=7.000000,Z=8.000000),WindAmplitude=(X=200.000000,Y=-500.000000,Z=400.000000),GravityOverride=(X=0.000000,Y=-300.000000,Z=1500.000000),JointName="root")),Skydive_FloatingDown=(GravityOverride=(WindFrequency=(X=1.500000,Y=7.000000,Z=8.000000),WindAmplitude=(X=200.000000,Y=-500.000000,Z=400.000000),WindOffset=(X=0.000000,Y=-600.000000,Z=400.000000),GravityOverride=(X=0.000000,Y=-700.000000,Z=600.000000),JointName="root")),Skydive_DivingDown=(GravityOverride=(WindFrequency=(X=1.500000,Y=12.000000,Z=10.000000),WindAmplitude=(X=50.000000,Y=300.000000,Z=100.000000),GravityOverride=(X=-1500.000000,Y=0.000000,Z=0.000000),JointName="pelvis")),Skydive_Parachuting=(GravityOverride=(WindFrequency=(X=1.500000,Y=7.000000,Z=8.000000),WindAmplitude=(X=200.000000,Y=-500.000000,Z=400.000000),GravityOverride=(X=0.000000,Y=-800.000000,Z=-980.000000),JointName="root")),OnGround_Standing=(GravityOverride=(GravityOverride=(X=0.000000,Y=0.000000,Z=-980.000000),JointName="root")),OnGround_Moving=(GravityOverride=(GravityOverride=(X=0.000000,Y=0.000000,Z=-980.000000),JointName="root")),OnGround_Moving_HighRotationRate=(GravityOverride=(GravityOverride=(X=0.000000,Y=0.000000,Z=-980.000000),JointName="root")),OnGround_Crouch_Moving=(GravityOverride=(GravityOverride=(X=0.000000,Y=0.000000,Z=-980.000000),JointName="root")),Emote_or_Melee=(GravityOverride=(GravityOverride=(X=0.000000,Y=0.000000,Z=-980.000000),JointName="root")),Swimming=(GravityOverride=(GravityOverride=(X=-800.000000,Y=0.000000,Z=0.000000),JointName="pelvis")),Front_End=(GravityOverride=(GravityOverride=(X=0.000000,Y=0.000000,Z=-980.000000),JointName="root")),DBNO=(GravityOverride=(GravityOverride=(X=-980.000000,Y=400.000000,Z=0.000000),JointName="pelvis")),DBNO_Carried=(GravityOverride=(GravityOverride=(X=-980.000000,Y=500.000000,Z=-500.000000),JointName="pelvis")),OnGround_Falling=(GravityOverride=(GravityOverride=(X=0.000000,Y=0.000000,Z=-980.000000),JointName="root")),Ziplining=(GravityOverride=(GravityOverride=(X=0.000000,Y=0.000000,Z=-980.000000),JointName="root")),Skydive_Surfing=(GravityOverride=(WindFrequency=(X=1.500000,Y=7.000000,Z=8.000000),WindAmplitude=(X=200.000000,Y=-500.000000,Z=400.000000),GravityOverride=(X=500.000000,Y=-300.000000,Z=-980.000000),JointName="root")),Riding_Standing=(GravityOverride=(GravityOverride=(X=0.000000,Y=-350.000000,Z=-980.000000),JointName="root")),Riding_Moving=(GravityOverride=(GravityOverride=(X=0.000000,Y=-450.000000,Z=-450.000000),JointName="root")),Riding_Moving_Targeting=(GravityOverride=(GravityOverride=(X=0.000000,Y=-200.000000,Z=-980.000000),JointName="root")),Tactical_Sprint=(GravityOverride=(GravityOverride=(X=0.000000,Y=-450.000000,Z=-500.000000),JointName="root")),Rail_Grinding=(GravityOverride=(GravityOverride=(X=0.000000,Y=0.000000,Z=-980.000000),JointName="root")),)
 }
 Head Rot {type=number}
 isAssaultRifleEquipped float {type=number}
 Pawn {type=FortPlayerPawn}
 RBAN Params {
  type=FortRigidBodyAnimNodeParameterSet
  default=(Skydive_GoingUp=(RigidBodyAnimNode=(SimSpaceSettings=(WorldAlpha=1.000000,MaxLinearVelocity=500.000000,MaxAngularVelocity=3.000000,MaxLinearAcceleration=800.000000,MaxAngularAcceleration=1.000000,ExternalLinearDragV=(X=1.000000,Y=1.000000,Z=1.000000)))),Skydive_FloatingDown=(RigidBodyAnimNode=(SimSpaceSettings=(WorldAlpha=1.000000,MaxLinearVelocity=500.000000,MaxAngularVelocity=3.000000,MaxLinearAcceleration=800.000000,MaxAngularAcceleration=1.000000,ExternalLinearDragV=(X=1.000000,Y=1.000000,Z=1.000000)))),Skydive_DivingDown=(RigidBodyAnimNode=(SimSpaceSettings=(WorldAlpha=1.000000,MaxLinearVelocity=500.000000,MaxAngularVelocity=3.000000,MaxLinearAcceleration=800.000000,MaxAngularAcceleration=1.000000,ExternalLinearDragV=(X=1.000000,Y=1.000000,Z=1.000000)))),Skydive_Parachuting=(RigidBodyAnimNode=(SimSpaceSettings=(WorldAlpha=1.000000,MaxLinearVelocity=500.000000,MaxAngularVelocity=3.000000,MaxLinearAcceleration=800.000000,MaxAngularAcceleration=1.000000,ExternalLinearDragV=(X=1.000000,Y=1.000000,Z=1.000000)))),OnGround_Standing=(RigidBodyAnimNode=(SimSpaceSettings=(WorldAlpha=1.000000,MaxLinearVelocity=500.000000,MaxAngularVelocity=3.000000,MaxLinearAcceleration=800.000000,MaxAngularAcceleration=1.000000,ExternalLinearDragV=(X=1.000000,Y=1.000000,Z=1.000000)))),OnGround_Moving=(RigidBodyAnimNode=(SimSpaceSettings=(WorldAlpha=1.000000,MaxLinearVelocity=500.000000,MaxAngularVelocity=3.000000,MaxLinearAcceleration=800.000000,MaxAngularAcceleration=1.000000,ExternalLinearDragV=(X=1.000000,Y=1.000000,Z=1.000000)))),OnGround_Moving_HighRotationRate=(RigidBodyAnimNode=(SimSpaceSettings=(WorldAlpha=1.000000,MaxLinearVelocity=500.000000,MaxAngularVelocity=3.000000,MaxLinearAcceleration=800.000000,MaxAngularAcceleration=1.000000,ExternalLinearDragV=(X=1.000000,Y=1.000000,Z=1.000000)))),OnGround_Crouch_Moving=(RigidBodyAnimNode=(SimSpaceSettings=(WorldAlpha=1.000000,MaxLinearVelocity=500.000000,MaxAngularVelocity=3.000000,MaxLinearAcceleration=800.000000,MaxAngularAcceleration=1.000000,ExternalLinearDragV=(X=1.000000,Y=1.000000,Z=1.000000)))),Emote_or_Melee=(RigidBodyAnimNode=(SimSpaceSettings=(WorldAlpha=1.000000,MaxLinearVelocity=500.000000,MaxAngularVelocity=3.000000,MaxLinearAcceleration=1000.000000,MaxAngularAcceleration=1.000000,ExternalLinearDragV=(X=1.000000,Y=1.000000,Z=1.000000)))),Swimming=(RigidBodyAnimNode=(SimSpaceSettings=(WorldAlpha=1.000000,MaxLinearVelocity=500.000000,MaxAngularVelocity=3.000000,MaxLinearAcceleration=800.000000,MaxAngularAcceleration=1.000000,ExternalLinearDragV=(X=1.000000,Y=1.000000,Z=1.000000)))),Front_End=(RigidBodyAnimNode=(SimSpaceSettings=(WorldAlpha=1.000000,MaxLinearVelocity=500.000000,MaxAngularVelocity=3.000000,MaxLinearAcceleration=800.000000,MaxAngularAcceleration=1.000000,ExternalLinearDragV=(X=1.000000,Y=1.000000,Z=1.000000)))),DBNO=(RigidBodyAnimNode=(SimSpaceSettings=(WorldAlpha=1.000000,MaxLinearVelocity=500.000000,MaxAngularVelocity=3.000000,MaxLinearAcceleration=800.000000,MaxAngularAcceleration=1.000000,ExternalLinearDragV=(X=1.000000,Y=1.000000,Z=1.000000)))),DBNO_Carried=(RigidBodyAnimNode=(SimSpaceSettings=(WorldAlpha=1.000000,MaxLinearVelocity=500.000000,MaxAngularVelocity=3.000000,MaxLinearAcceleration=800.000000,MaxAngularAcceleration=1.000000,ExternalLinearDragV=(X=1.000000,Y=1.000000,Z=1.000000)))),OnGround_Falling=(RigidBodyAnimNode=(SimSpaceSettings=(WorldAlpha=1.000000,MaxLinearVelocity=500.000000,MaxAngularVelocity=3.000000,MaxLinearAcceleration=800.000000,MaxAngularAcceleration=1.000000,ExternalLinearDragV=(X=1.000000,Y=1.000000,Z=1.000000)))),Ziplining=(RigidBodyAnimNode=(SimSpaceSettings=(WorldAlpha=1.000000,MaxLinearVelocity=500.000000,MaxAngularVelocity=3.000000,MaxLinearAcceleration=800.000000,MaxAngularAcceleration=1.000000,ExternalLinearDragV=(X=1.000000,Y=1.000000,Z=1.000000)))),Skydive_Surfing=(RigidBodyAnimNode=(SimSpaceSettings=(WorldAlpha=1.000000,MaxLinearVelocity=500.000000,MaxAngularVelocity=3.000000,MaxLinearAcceleration=800.000000,MaxAngularAcceleration=1.000000,ExternalLinearDragV=(X=1.000000,Y=1.000000,Z=1.000000)))),Riding_Standing=(RigidBodyAnimNode=(SimSpaceSettings=(WorldAlpha=1.000000,MaxLinearVelocity=500.000000,MaxAngularVelocity=3.000000,MaxLinearAcceleration=800.000000,MaxAngularAcceleration=1.000000,ExternalLinearDragV=(X=1.000000,Y=1.000000,Z=1.000000)))),Riding_Moving=(RigidBodyAnimNode=(SimSpaceSettings=(WorldAlpha=1.000000,MaxLinearVelocity=500.000000,MaxAngularVelocity=3.000000,MaxLinearAcceleration=800.000000,MaxAngularAcceleration=1.000000,ExternalLinearDragV=(X=1.000000,Y=1.000000,Z=1.000000)))),Riding_Moving_Targeting=(RigidBodyAnimNode=(SimSpaceSettings=(WorldAlpha=1.000000,MaxLinearVelocity=500.000000,MaxAngularVelocity=3.000000,MaxLinearAcceleration=1000.000000,MaxAngularAcceleration=1.000000,ExternalLinearDragV=(X=1.000000,Y=1.000000,Z=1.000000)))),Tactical_Sprint=(RigidBodyAnimNode=(SimSpaceSettings=(WorldAlpha=1.000000,MaxLinearVelocity=500.000000,MaxAngularVelocity=3.000000,MaxLinearAcceleration=800.000000,MaxAngularAcceleration=1.000000,ExternalLinearDragV=(X=1.000000,Y=1.000000,Z=1.000000)))),Rail_Grinding=(RigidBodyAnimNode=(SimSpaceSettings=(WorldAlpha=1.000000,MaxLinearVelocity=500.000000,MaxAngularVelocity=3.000000,MaxLinearAcceleration=800.000000,MaxAngularAcceleration=1.000000,ExternalLinearDragV=(X=1.000000,Y=1.000000,Z=1.000000)))),)
 }
 RBAN Settings Out {type=FortRigidBodyAnimNodeParameters}
 SkelMesh {type=SkeletalMeshComponent}
]
